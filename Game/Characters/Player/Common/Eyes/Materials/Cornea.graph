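# Cornea material for the common player character eyes, using the Preintegrated Skin shading model.
# It is enabled for skeletal meshes, morph targets and mesh deformers.
# The graph holds a single vector parameter with a mid-gray, zero-alpha default and shows no connection from it to any material output.
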Material /Game/Characters/Player/Common/Eyes/Materials/Cornea
graph {
 properties {
  bUsedWithMeshDeformer=true
  bUsedWithMorphTargets=true
  bUsedWithSkeletalMesh=true
  pixelDepthOffsetMode=PDOM_Legacy
  shadingModel=MSM_PreintegratedSkin
  usedShadingModels=Preintegrated Skin
 }
 nodes [VectorParameter() {defaultValue=<rgba 0.5 0.5 0.5 0>}]
}
