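# Brown eye material for player characters, a masked Default Lit surface usable on skeletal meshes with morph targets and mesh deformers.
# It combines an Eye_Brown texture sample, two scalar constants and the shared CharacterResInEffect material function.
# That function exposes Animation Gradient and Rez In Percentage for a character rez-in effect.
# The listing shows no links between these nodes or to the material outputs.
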Material /Game/Characters/Player/Common/Eyes/Materials/Eye_Brown
graph {
 properties {
  bUsedWithMeshDeformer=true
  bUsedWithMorphTargets=true
  bUsedWithSkeletalMesh=true
  blendMode=BLEND_Masked
  pixelDepthOffsetMode=PDOM_Legacy
  usedShadingModels=Default Lit
 }
 nodes [
  Constant() {r=0.1333}
  Constant() {r=0.3}
  MaterialFunctionCall!(Animation Gradient, Rez In Percentage) {materialFunction=/Game/Packages/Fortress_SharedMaterials/Base_Material_Functions/CharacterResInEffect}
  TextureSample() {texture=/Game/Characters/Player/Common/Eyes/Textures/Eye_Brown}
 ]
}
