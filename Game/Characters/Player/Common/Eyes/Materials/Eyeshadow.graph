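# Unlit Modulate-blend material for the player character eyeshadow, darkening what lies behind it on skeletal, morph-target meshes.
# It lerps from white, which leaves the background unchanged, to a near-black reddish tint parameter.
# The blend alpha is vertex color red times the Modulative Opacity output of the shared CharacterResInEffect function, tying the shading to the character res-in effect.
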Material /Game/Characters/Player/Common/Eyes/Materials/Eyeshadow
graph {
 properties {
  bUsedWithMeshDeformer=true
  bUsedWithMorphTargets=true
  bUsedWithNanite=true
  bUsedWithSkeletalMesh=true
  blendMode=BLEND_Modulate
  pixelDepthOffsetMode=PDOM_Legacy
  shadingModel=MSM_Unlit
  translucencyPass=MTP_BeforeDOF
  usedShadingModels=Unlit
 }
 nodes [LinearInterpolate(
   B=VectorParameter() {defaultValue=<rgba 0.006 0.0009 0.0006 0>}.RGB
   Alpha=Multiply(
    A=VertexColor().R
    B=MaterialFunctionCall!(Animation Gradient, Rez In Percentage) {materialFunction=/Game/Packages/Fortress_SharedMaterials/Base_Material_Functions/CharacterResInEffect}.Modulative Opacity
   )
  ) {constA=1}]
}
